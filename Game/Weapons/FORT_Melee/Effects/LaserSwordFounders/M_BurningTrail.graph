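# Translucent, unlit, two-sided material for laser sword trails on beam trails and Niagara ribbons.
# Particle color is scaled by time of day and passed through FX emissive adjustments.
# A mask product combines particle alpha, a U-axis falloff, the trail texture and UV1-panned noise shaped by TexturePower and OpacityBoost dynamic parameters.
# A camera-to-normal dot product fades that product where the ribbon turns edge-on.
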Material /Game/Weapons/FORT_Melee/Effects/LaserSwordFounders/M_BurningTrail
graph {
 properties {
  bUsedWithBeamTrails=true
  bUsedWithNiagaraRibbons=true
  blendMode=BLEND_Translucent
  pixelDepthOffsetMode=PDOM_Legacy
  shadingModel=MSM_Unlit
  twoSided=true
  usedShadingModels=Unlit
 }
 nodes [
  Constant()
  n1 = ParticleColor()
  n2 = Multiply(A=@n1.RGB) {constB=1.5}
  MaterialFunctionCall(Emissive=MaterialFunctionCall(Color_Morning=@n2, Color_Day=Multiply(A=@n1.RGB) {constB=2.5}, Color_Evening=@n2, Color_Night=@n1.RGB) {materialFunction=&Color_Time_of_Day}.Result) {materialFunction=&MF_FX_EmissiveAdjustments}
  n3 = OneMinus(None=ComponentMask(None=TextureCoordinate()) {r=true})
  n4 = Multiply(A=@n1.A, B=OneMinus(None=Power(Base=@n3) {constExponent=4}))
  n5 = TextureSampleParameter2D() {parameterName=TrailMask, texture=&/T_LaserSwordTrail}
  n6 = TextureCoordinate() {coordinateIndex=1}
  n7 = Add(
   A=TextureSampleParameter2D(UVs=Panner(Coordinate=@n6) {speedX=-3}) {parameterName=TrailScrollTexture, texture=&/T_LaserSwordTrailNoise}.RGB
   B=DynamicParameter() {defaultValue=<rgba 1 0 1 1>, paramNames [TexturePower,OpacityBoost,Param3,Param4]}.OpacityBoost
  )
  n8 = DynamicParameter() {defaultValue=<rgba 1 0 1 1>, paramNames [TexturePower,OpacityBoost,Param3,Param4]}
  n9 = DotProduct(A=Normalize(VectorInput=CameraVectorWS()), B=Normalize(VectorInput=VertexNormalWS()))
  Multiply(
   A=Multiply(A=Multiply(A=@n4, B=@n5.RGB), B=Power(Base=@n7, Exp=@n8.TexturePower) {constExponent=0})
   B=Power(Base=Abs(None=@n9), Exp=Constant() {r=3})
  )
 ]
}
